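; Controller component that configures Sparks song (Jam) emote options, deciding where Jam categories sit relative to Locker emotes.
; On BeginPlay, outside the Campaign subgame, it asynchronously waits for playlist data and caches the context tags, while BeginSetup runs without waiting.
; Those tags identify Jam playlists by matching a configured tag set.
; Jam categories precede Locker emotes only in Jam playlists, unless a flag places them after.
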(event EventBeginPlay
  (|CallParentFunction)
  (if (not (FortGlobals|IsinSubGame "Campaign"))
    (bind (_gamestate _playlist _playlistcontexttags) (Class|FortAsyncActionCurrentPlaylistReady|PlaylistDataReadyAsync)
      (:OnReady
        (Variables|Default|SetPlaylist%20Context%20Tags _playlistcontexttags))))
  (CallFunction|BeginSetup))

(fn IsJamPlaylist
  (return (GameplayTags|HasAnyTags (Variables|Default|GetPlaylist%20Context%20Tags) (Variables|Default|GetJam%20Emote%20First%20Playlist%20Tags) true)))

(fn ShouldAddJamCategoriesBeforeLocker
  ;[0.315 0.2498 0.05414] Always go after Locker emotes if we are not in a Jam playlist
  (if (CallFunction|IsJamPlaylist)
    ;[0.5 0.15 0.15 0.5] We are in a Jam playlist, return the set value
    (return (not (Variables|Default|GetbJamEmotesPostLockerInJamPlaylists)))
    (else
      ;[0.5 0.15 0.15 0.5] We are not in a Jam playlist, return false
      (return false))))
